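# Parent character material for the NobleCrusher skin, built on shared character functions and reusing SnowKnight color customization, cloth fuzz and banner logo placement.
# Channel masks drive tints, eye glow and optional metal LUT shading.
# Primary/secondary masks set Ice, Galaxy and Chroma alphas for advanced materials, with optional crystal before PostFX.
# Galaxy uses triplanar virtual-box nebulae and stars, with time-of-day adjustment and one nebula layer on lower mobile tiers.
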Material /BRCosmetics/Characters/Player/Male/Medium/Bodies/M_MED_NobleCrusher/Materials/M_NobleCrusher_PARENT
graph {
 properties {
  bAutomaticallySetUsageInEditor=false
  bUseMaterialAttributes=true
  bUsedWithClothing=true
  bUsedWithMeshDeformer=true
  bUsedWithMorphTargets=true
  bUsedWithNiagaraMeshParticles=true
  bUsedWithNiagaraRibbons=true
  bUsedWithNiagaraSprites=true
  bUsedWithSkeletalMesh=true
  pixelDepthOffsetMode=PDOM_Legacy
  shadingModel=MSM_FromMaterialExpression
  usedShadingModels=Default Lit | Cloth
 }
 nodes [
  n1 = StaticSwitch(True=NamedRerouteUsage(), False=NamedRerouteUsage(), Value=StaticBoolParameter() {parameterName=UseAdvancedMats})
  n2 = NamedRerouteUsage()
  MaterialFunctionCall(
   Material=StaticSwitchParameter(
    True=MaterialFunctionCall!(Material=@n1, TechArtMask=@n2) {materialFunction=&/MF_NewDiamond_Low}.Material
    False=@n1
   ) {group=Skin FX, parameterName=Use Crystal, sortPriority=0}
   M=NamedRerouteUsage()
   Pre Skinned Local Position=NamedRerouteUsage()
   Distance Blend=NamedRerouteUsage()
   Fresnel=NamedRerouteUsage()
  ) {materialFunction=&MF_CharacterParent_2023_PostFX}
  n3 = GetMaterialAttributes(None=MaterialFunctionCall!(Material=NamedRerouteUsage(), Ao/Gmap/Skin=NamedRerouteUsage(), Fresnel=NamedRerouteUsage(), Alpha=NamedRerouteUsage()) {materialFunction=/Game/Characters/Player/Female/Medium/Bodies/F_MED_SnowKnight_Helm/Materials/MF_SnowKnight_ClothFuzz}.Result) {attributeGetTypes [#1]}
  NamedRerouteDeclaration(None=@n3.MaterialAttributes) {name=MaterialAttributes, nodeColor=<rgba 0.9219 0 1 1>}
  n4 = TextureSampleParameter2D() {
   group=05_Tech Art Masks
   parameterName=TechArtMask
   samplerSource=SSM_Wrap_WorldGroupSettings
   samplerType=SAMPLERTYPE_Masks
   sortPriority=0
   texture=&T_BlackMask
  }
  NamedRerouteDeclaration(None=@n4.RGB) {name=TechArt/FX, nodeColor=<rgba 0.0652 0.0652 0.9635 1>}
  NamedRerouteDeclaration(None=NamedRerouteUsage()) {name=MaskCrystal, nodeColor=<rgba 0.0652 0.0652 0.9635 1>}
  NamedRerouteDeclaration(None=Fresnel(ExponentIn=Constant() {r=1}) {baseReflectFraction=0}) {name=Fresnel, nodeColor=<rgba 0 0.4609 1 1>}
  n5 = PreSkinnedPosition()
  NamedRerouteDeclaration(None=ShaderStageSwitch(PixelShader=VertexInterpolator(VS=@n5).PS, VertexShader=@n5)) {name=PreskinnedPos, nodeColor=<rgba 1 0 0.8906 1>}
  n6 = MaterialFunctionCall!(UV) {materialFunction=/Game/Global/Materials/Character_Parent_2023/Functions/MF_CharacterParent_2023_TextureInputs}
  n7 = MaterialFunctionCall!(Material=@n6.Result) {materialFunction=/Game/Packages/Fortress_SharedMaterials/Base_Material_Functions/CharacterBrightnessandScale_Balance}
  NamedRerouteDeclaration(None=@n7.Distance Blend) {name=DistanceBlend, nodeColor=<rgba 1 0.8906 0 1>}
  NamedRerouteDeclaration(None=LinearInterpolate(
    A=NamedRerouteUsage()
    Alpha=ScalarParameter() {controlType=Enumeration, enumeration=/Game/Global/Materials/Enums/E_Mat_YesNo, parameterName=Turn Off Primary Mask}
   ) {constB=0}) {name=PrimaryMask, nodeColor=<rgba 1 0.8203 0 1>}
  n8 = NamedRerouteUsage()
  NamedRerouteDeclaration(None=Multiply(A=NamedRerouteUsage(), B=OneMinus(None=@n8))) {name=SecondaryMask, nodeColor=<rgba 0 1 0.7422 1>}
  n9 = Multiply(A=ScalarParameter() {parameterName=PrimaryIce}, B=NamedRerouteUsage())
  n10 = Multiply(A=ScalarParameter() {parameterName=SecondaryIce}, B=NamedRerouteUsage())
  NamedRerouteDeclaration(None=Saturate(None=Add(A=@n9, B=@n10))) {name=IceAlpha, nodeColor=<rgba 0.7578 0 1 1>}
  n11 = Multiply(A=ScalarParameter() {parameterName=PrimaryGalaxy}, B=NamedRerouteUsage())
  n12 = Multiply(A=ScalarParameter() {parameterName=SecondaryGalaxy}, B=NamedRerouteUsage())
  NamedRerouteDeclaration(None=Saturate(None=Add(A=@n11, B=@n12))) {name=GalaxyAlpha, nodeColor=<rgba 0 0.2969 1 1>}
  n13 = Multiply(A=ScalarParameter() {parameterName=PrimaryChroma}, B=NamedRerouteUsage())
  n14 = Multiply(A=ScalarParameter() {parameterName=SecondaryChroma}, B=NamedRerouteUsage())
  NamedRerouteDeclaration(None=Saturate(None=Add(A=@n13, B=@n14))) {name=ChromaAlpha, nodeColor=<rgba 0 0.2969 1 1>}
  NamedRerouteDeclaration(None=MaterialFunctionCall!(Material Attributes=NamedRerouteUsage(), Fresnel=NamedRerouteUsage(), Alpha=NamedRerouteUsage()) {materialFunction=/Game/Global/MaterialFunctions/TechArt/MF_ThinFilm_SnowKnight}.Result) {name=AdvMat, nodeColor=<rgba 0 1 0.3203 1>}
  NamedRerouteDeclaration(None=StaticSwitchParameter(True=NamedRerouteUsage(), False=NamedRerouteUsage()) {group=01_SnowKnight_ColorCustomization, parameterName=UseSnowKnight_ColorCustomization, sortPriority=0}) {name=Out_Material_CustomColor, nodeColor=<rgba 0.1484 0 1 1>}
  NamedRerouteDeclaration(None=MaterialFunctionCall!(In=NamedRerouteUsage(), InMask=NamedRerouteUsage(), InFXMask=NamedRerouteUsage()) {materialFunction=/Game/Global/Materials/MaterialGym/Materials/SimpleIce/MF_SimpleIce}.Result) {name=Out_Ice_Mat, nodeColor=<rgba 0.0001 0.9743 1 1>}
  n15 = TextureObjectParameter() {
   group=Galaxy
   parameterName=StarsTexture
   samplerType=SAMPLERTYPE_Masks
   sortPriority=80
   texture=/Game/Characters/Player/Male/Medium/Bodies/M_Med_Soldier_04/Skins/Galaxy/Textures/Celestial_T_Stars
  }
  n16 = MaterialFunctionCall!(
   Texture=@n15
   Depth=ScalarParameter() {defaultValue=150, group=Galaxy, parameterName=Stars_Depth, sortPriority=81}
   Tile=ScalarParameter() {defaultValue=2, group=Galaxy, parameterName=Stars_Tile, sortPriority=82}
   BlendContrast=ScalarParameter() {defaultValue=1, group=Galaxy, parameterName=Stars_BlendContrast, sortPriority=83}
   SpeedXY=ComponentMask(None=VectorParameter() {group=Galaxy, parameterName=Stars_Speed, sortPriority=84, defaultValue=<rgba 0.002 -0.03 0 1>}.RGB) {g=true, r=true}
   ActorOffset=NamedRerouteUsage()
   bUseLocalCameraVector
   bUseDitherSingleSample
  ) {desc=3 TextureSampler, materialFunction=/Game/Global/MaterialFunctions/MF_TriplanarVirtualBox}
  NamedRerouteDeclaration(None=Multiply(
    A=ComponentMask(None=@n16.Result) {b=true}
    B=ScalarParameter() {defaultValue=1, group=Galaxy, parameterName=Stars_Brightness, sortPriority=84}
   )) {name=Stars, nodeColor=<rgba 0 1 0.5078 1>}
  n17 = Multiply(
   A=Fresnel(ExponentIn=ScalarParameter() {defaultValue=3, group=Galaxy, parameterName=EmissiveFresnelExponent, sortPriority=90}) {baseReflectFraction=0}
   B=ScalarParameter() {defaultValue=1, group=Galaxy, parameterName=EmissiveFresnelIntensity, sortPriority=92}
  )
  NamedRerouteDeclaration(None=Multiply(
    A=NamedRerouteDeclaration(None=@n17) {name=FresnelMask, nodeColor=<rgba 0 0.7891 1 1>}
    B=VectorParameter() {group=Galaxy, parameterName=EmissiveColor, sortPriority=91, defaultValue=<rgba 0.0885 0.6921 1 1>}.RGB
   )) {name=Fresnel, nodeColor=<rgba 0.8047 0 1 1>}
  n18 = TextureObjectParameter() {
   group=Galaxy
   parameterName=Galaxy_01_Texture
   sortPriority=58
   texture=/Game/Characters/Player/Male/Medium/Bodies/M_Med_Soldier_04/Skins/Galaxy/Textures/T_FN_Nebula_style_blue
  }
  n19 = ScalarParameter() {defaultValue=80, group=Galaxy, parameterName=Galaxy_01_Depth, sortPriority=61}
  n20 = ScalarParameter() {defaultValue=0.4, group=Galaxy, parameterName=Galaxy_01_Tile, sortPriority=62}
  n21 = ScalarParameter() {defaultValue=0.4, group=Galaxy, parameterName=Galaxy_01_BlendContrast, sortPriority=63}
  n22 = VectorParameter() {group=Galaxy, parameterName=Galaxy_01_Speed, sortPriority=64, defaultValue=<rgba 0.002 -0.0025 0 1>}
  n23 = ComponentMask(None=@n22.RGB) {g=true, r=true}
  n24 = NamedRerouteDeclaration(None=VectorParameter() {group=Galaxy, parameterName=Galaxy_Offset, sortPriority=59, defaultValue=<rgba 0 0 0 1>}.RGB) {name=ActorOffset, nodeColor=<rgba 0.9688 1 0 1>}
  n25 = Multiply(
   A=MaterialFunctionCall(
    Hue Shift Percentage=ScalarParameter() {group=Galaxy, parameterName=Galaxy_01_HueShift, sortPriority=64}
    Texture=MaterialFunctionCall(Texture=@n18, Depth=@n19, Tile=@n20, BlendContrast=@n21, SpeedXY=@n23, ActorOffset=@n24) {materialFunction=&MF_TriplanarVirtualBox}.Result
   ) {materialFunction=&HueShift}.Result
   B=ScalarParameter() {defaultValue=1, group=Galaxy, parameterName=Galaxy_01_Brightness, sortPriority=65}
  )
  n26 = ScalarParameter() {defaultValue=1.75, group=Galaxy, parameterName=Galaxy_01_Exp, sortPriority=66}
  NamedRerouteDeclaration(None=Desaturation(
    None=Power(Base=@n25, Exp=@n26)
    Fraction=ScalarParameter() {defaultValue=0.5, group=Galaxy, parameterName=Galaxy_01_Desaturation, sortPriority=67}
   )) {name=Galaxy_01, nodeColor=<rgba 1 0.2578 0 1>}
  n27 = OneMinus(None=NamedRerouteUsage())
  n28 = NamedRerouteUsage()
  NamedRerouteDeclaration(None=Add(A=Multiply(A=@n27, B=@n28), B=NamedRerouteUsage())) {name=OUT_Emissive, nodeColor=<rgba 1 0 0.1875 1>}
  n29 = SetMaterialAttributes!(
   MaterialAttributes=NamedRerouteUsage()
   Base Color=NamedRerouteUsage()
   Metallic=NamedRerouteUsage()
   Specular=NamedRerouteUsage()
   Roughness=NamedRerouteUsage()
   Emissive Color=NamedRerouteUsage()
  ) {attributeSetTypes [
    #1
    #2
    #3
    #4
    #5
   ]}
  NamedRerouteDeclaration(None=@n29) {name=MaterialOUT, nodeColor=<rgba 0 1 0.625 1>}
  n30 = ScalarParameter() {group=Galaxy, parameterName=Galaxy_02_HueShift, sortPriority=74}
  n31 = TextureObjectParameter() {
   group=Galaxy
   parameterName=Galaxy_02_Texture
   sortPriority=70
   texture=/Game/Characters/Player/Male/Medium/Bodies/M_Med_Soldier_04/Skins/Galaxy/Textures/T_FN_Nebula_style2
  }
  n32 = VectorParameter() {group=Galaxy, parameterName=Galaxy_02_Speed, sortPriority=74, defaultValue=<rgba -0.002 -0.0065 0 1>}
  n33 = MaterialFunctionCall(
   Texture=@n31
   Depth=ScalarParameter() {defaultValue=35, group=Galaxy, parameterName=Galaxy_02_Depth, sortPriority=71}
   Tile=ScalarParameter() {defaultValue=0.4, group=Galaxy, parameterName=Galaxy_02_Tile, sortPriority=72}
   BlendContrast=ScalarParameter() {defaultValue=1, group=Galaxy, parameterName=Galaxy_02_BlendContrast, sortPriority=73}
   SpeedXY=ComponentMask(None=@n32.RGB) {g=true, r=true}
   ActorOffset=NamedRerouteUsage()
  ) {materialFunction=&MF_TriplanarVirtualBox}
  NamedRerouteDeclaration(None=Multiply(
    A=MaterialFunctionCall(Hue Shift Percentage=@n30, Texture=@n33.Result) {materialFunction=&HueShift}.Result
    B=ScalarParameter() {defaultValue=0.5, group=Galaxy, parameterName=Galaxy_02_Brightness, sortPriority=75}
   )) {name=Galaxy_02, nodeColor=<rgba 0.0781 0 1 1>}
  n34 = NamedRerouteUsage()
  n35 = Add(A=@n34, B=NamedRerouteUsage())
  n36 = MaterialFunctionCall!(
   Low _ PC Low and Med=@n35
   "High _ XB1, PS4"=@n35
   "Epic _ PC High, XSX, PS5"=@n35
   --------
   "Mobile Low _ PC Perf, Android"=@n34
   Mobile Med _ Switch=@n34
  ) {materialFunction=/Game/Global/MaterialFunctions/TechArt/MF_Fortnite_QualitySwitch_Vector3}
  n37 = ChannelMaskParameter(None=NamedRerouteDeclaration(None=Saturate(None=@n36.Result)) {name=Galaxies, nodeColor=<rgba 1 0 0.2578 1>}) {group=Galaxy, maskChannel=Blue, parameterName=StarMaskChannel, sortPriority=84, defaultValue=<rgba 0 0 1 0>}
  NamedRerouteDeclaration(None=Multiply(A=@n37.R, B=ScalarParameter() {defaultValue=3, group=Galaxy, parameterName=GalaxyStarMaskStrength, sortPriority=85}) {constB=2}) {name=StarMask, nodeColor=<rgba 0.9922 1 0 1>}
  n38 = LinearInterpolate(A=NamedRerouteUsage(), Alpha=NamedRerouteUsage()) {constB=0}
  n39 = Multiply(
   A=NamedRerouteUsage()
   B=MaterialFunctionCall(
    Value_Morning=ScalarParameter() {defaultValue=1, group=Galaxy, parameterName=Galaxy_Brightness_Morning, sortPriority=100}
    Value_Day=ScalarParameter() {defaultValue=1, group=Galaxy, parameterName=Galaxy_Brightness_Day, sortPriority=101}
    Value_Evening=ScalarParameter() {defaultValue=1, group=Galaxy, parameterName=Galaxy_Brightness_Evening, sortPriority=102}
    Value_Night=ScalarParameter() {defaultValue=1, group=Galaxy, parameterName=Galaxy_Brightness_Night, sortPriority=103}
   ) {materialFunction=&Scalar_Time_of_Day}.Result
  )
  n40 = MaterialFunctionCall(
   Value_Morning=ScalarParameter() {group=Galaxy, parameterName=Galaxy_Desat_Morning, sortPriority=110}
   Value_Day=ScalarParameter() {group=Galaxy, parameterName=Galaxy_Desat_Day, sortPriority=111}
   Value_Evening=ScalarParameter() {group=Galaxy, parameterName=Galaxy_Desat_Evening, sortPriority=112}
   Value_Night=ScalarParameter() {group=Galaxy, parameterName=Galaxy_Desat_Night, sortPriority=113}
  ) {materialFunction=&Scalar_Time_of_Day}
  NamedRerouteDeclaration(None=Add(A=Saturate(None=@n38), B=Desaturation(None=@n39, Fraction=@n40.Result))) {name=GalaxiesAndStars, nodeColor=<rgba 0 1 0.9062 1>}
  n41 = GetMaterialAttributes(None=NamedRerouteUsage()) {attributeGetTypes [
    #4
    #2
    #3
    #5
    #1
   ]}
  n42 = NamedRerouteDeclaration(None=ScalarParameter() {group=Galaxy, parameterName=UseIncomingMatParams}) {name=UseIncomingMatParams, nodeColor=<rgba 0.875 0 1 1>}
  NamedRerouteDeclaration(None=LinearInterpolate(A=ScalarParameter() {defaultValue=0.2, group=Galaxy, parameterName=Galaxy_Roughness, sortPriority=52}, B=@n41.Roughness, Alpha=@n42) {constB=0}) {name=OUTRoughness, nodeColor=<rgba 0 0.4844 1 1>}
  NamedRerouteDeclaration(None=LinearInterpolate(A=ScalarParameter() {defaultValue=0.1, group=Galaxy, parameterName=Galaxy_Metallic, sortPriority=50}, B=@n41.Metallic, Alpha=@n42) {constB=0}) {name=OUTMetallic, nodeColor=<rgba 0 0.4844 1 1>}
  NamedRerouteDeclaration(None=LinearInterpolate(A=ScalarParameter() {defaultValue=0.5, group=Galaxy, parameterName=Galaxy_Spec, sortPriority=51}, B=@n41.Specular, Alpha=@n42) {constB=0}) {name=OUTSpec, nodeColor=<rgba 0 0.4844 1 1>}
  NamedRerouteDeclaration(None=@n41.Emissive Color) {name=IN_Emissive, nodeColor=<rgba 0.4766 1 0 1>}
  NamedRerouteDeclaration(None=LinearInterpolate(
    A=VectorParameter() {group=Galaxy, parameterName=Galaxy_BaseColor, sortPriority=49, defaultValue=<rgba 0 0 0 1>}.RGB
    B=@n41.Emissive Color
    Alpha=@n42
   ) {constB=0}) {name=OUTBaseColor, nodeColor=<rgba 1 0 0.6562 1>}
  NamedRerouteDeclaration(None=BlendMaterialAttributes(A=NamedRerouteUsage(), B=NamedRerouteUsage(), Alpha=NamedRerouteUsage())) {name=Galaxy_Mat, nodeColor=<rgba 1 0 0.9375 1>}
  n43 = MaterialFunctionCall(Float3=@n6.M) {desc="R = AO\r\nG = GMap\r\nB = Skin", materialFunction=&BreakOutFloat3Components}
  NamedRerouteDeclaration(None=@n43.R) {name=Ambient Occlusion, nodeColor=<rgba 0.875 0 1 1>}
  NamedRerouteDeclaration(None=@n43.G) {name=GMap, nodeColor=<rgba 0.6641 0 1 1>}
  NamedRerouteDeclaration(None=@n43.B) {name=Skin, nodeColor=<rgba 0.6641 0 1 1>}
  n44 = GetMaterialAttributes(None=@n6.Result) {attributeGetTypes [#6, #1]}
  NamedRerouteDeclaration(None=@n44.Normal) {name=Normal Map, nodeColor=<rgba 1 0.3281 0 1>}
  NamedRerouteDeclaration(None=@n6.M) {name=Mask, nodeColor=<rgba 0.3125 0 1 1>}
  NamedRerouteDeclaration(None=MaterialFunctionCall!(Material=@n7.Result, Emissive UVs, Fresnel=NamedRerouteUsage(), Alpha) {materialFunction=/Game/Global/Materials/Character_Parent_2023/Functions/MF_CharacterParent_2023_Emissive}.Result) {name=Material, nodeColor=<rgba 1 0 0.9844 1>}
  n45 = GetMaterialAttributes(None=NamedRerouteUsage()) {attributeGetTypes [#1, #5, #2]}
  NamedRerouteDeclaration(None=@n45.MaterialAttributes) {name=MA_In, nodeColor=<rgba 0 0.3438 1 1>}
  n46 = NamedRerouteUsage()
  n47 = NamedRerouteUsage()
  n48 = MaterialFunctionCall(Base=@n46, Blend=@n47) {materialFunction=&Blend_Overlay}
  n49 = Multiply(A=@n46, B=@n47)
  n50 = ScalarParameter() {group=01_SnowKnight_ColorCustomization, parameterName=OverlayTintBlend, sliderMax=1}
  n51 = NamedRerouteUsage()
  n52 = ScalarParameter() {controlType=Enumeration, defaultValue=1, enumeration=&E_Mat_YesNo, parameterName=Use Eye Glow}
  NamedRerouteDeclaration(None=SetMaterialAttributes!(
    MaterialAttributes=NamedRerouteUsage()
    Base Color=LinearInterpolate(A=@n48.Result, B=@n49, Alpha=@n50)
    Emissive Color=LinearInterpolate(B=@n51, Alpha=@n52)
   ) {attributeSetTypes [#1, #5]}) {name=MA_Out, nodeColor=<rgba 1 0.3008 0.2809 1>}
  n53 = NamedRerouteUsage()
  n54 = ChannelMaskParameter(None=@n53) {
   group=01_SnowKnight_ColorCustomization
   maskChannel=Green
   parameterName=TertiaryColor_ChannelMask
   sortPriority=6
   defaultValue=<rgba 0 1 0 0>
  }
  n55 = Max(
   A=NamedRerouteDeclaration(None=ChannelMaskParameter(None=@n53) {group=01_SnowKnight_ColorCustomization, parameterName=SecondaryColor_ChannelMask, sortPriority=5}.R) {name=SecondaryColor_Mask, nodeColor=<rgba 0.5938 0 1 1>}
   B=NamedRerouteDeclaration(None=@n54.R) {name=TertiaryColor_Mask, nodeColor=<rgba 0.5938 0 1 1>}
  )
  n56 = ChannelMaskParameter(None=@n53) {
   group=01_SnowKnight_ColorCustomization
   maskChannel=Blue
   parameterName=CrystalColor_ChannelMask
   sortPriority=7
   defaultValue=<rgba 0 0 1 0>
  }
  n57 = NamedRerouteDeclaration(None=@n56.R) {name=CrystalColor_Mask, nodeColor=<rgba 1 0.4219 0 1>}
  n58 = StaticSwitchParameter(True=NamedRerouteUsage(), False=Constant()) {parameterName=UseBanner}
  n59 = NamedRerouteDeclaration(None=Saturate(None=Add(A=@n57, B=@n58))) {name=CrystalBannerColor_Mask, nodeColor=<rgba 0.5938 0 1 1>}
  NamedRerouteDeclaration(None=OneMinus(None=Max(A=@n55, B=@n59))) {name=PrimaryColor_Mask, nodeColor=<rgba 0.5938 0 1 1>}
  NamedRerouteDeclaration(None=@n45.Base Color) {name=BaseColor_In, nodeColor=<rgba 0 0.3438 1 1>}
  n60 = VectorParameter() {group=01_SnowKnight_ColorCustomization, parameterName=Primary_Color, sortPriority=10, defaultValue=<rgba 1 1 1 1>}
  n61 = Multiply(A=NamedRerouteUsage(), B=ScalarParameter() {parameterName=Primary_ChainmailBoost})
  n62 = VectorParameter() {group=01_SnowKnight_ColorCustomization, parameterName=Secondary_Color, sortPriority=11, defaultValue=<rgba 1 0 0 1>}
  n63 = Multiply(A=NamedRerouteUsage(), B=ScalarParameter() {parameterName=Secondary_ChainmailBoost})
  n64 = LinearInterpolate(A=Multiply(A=@n60.RGB, B=Add(B=@n61) {constA=1}), B=Multiply(A=@n62.RGB, B=Add(B=@n63) {constA=1}), Alpha=NamedRerouteUsage())
  n65 = VectorParameter() {group=01_SnowKnight_ColorCustomization, parameterName=Trim_Color, sortPriority=11, defaultValue=<rgba 0 1 0 1>}
  n66 = NamedRerouteUsage()
  n67 = VectorParameter() {group=01_SnowKnight_ColorCustomization, parameterName=Crystal_Color, sortPriority=12, defaultValue=<rgba 0 0 1 1>}
  NamedRerouteDeclaration(None=LinearInterpolate(A=LinearInterpolate(A=@n64, B=@n65.RGB, Alpha=@n66), B=@n67.RGB, Alpha=NamedRerouteUsage())) {name=BaseColor_Out, nodeColor=<rgba 1 0.5859 0 1>}
  NamedRerouteDeclaration(None=@n45.Emissive Color) {name=Emissive_In, nodeColor=<rgba 0 0.3438 1 1>}
  n68 = NamedRerouteUsage()
  n69 = ChannelMaskParameter(None=@n68) {group=01_SnowKnight_ColorCustomization, maskChannel=Alpha, parameterName=FX_EyeMask, sortPriority=8, defaultValue=<rgba 0 0 0 1>}
  NamedRerouteDeclaration(None=@n69.R) {name=EyeGlowMask, nodeColor=<rgba 0 0.0391 1 1>}
  n70 = NamedRerouteUsage()
  NamedRerouteDeclaration(None=Multiply(
    A=Multiply(A=@n67.RGB, B=@n70)
    B=ScalarParameter() {defaultValue=5, group=01_SnowKnight_ColorCustomization, parameterName=EyeGlowEmissiveBrightness}
   )) {name=Emissive_Out, nodeColor=<rgba 1 0.5859 0 1>}
  n71 = TextureSampleParameter2D() {
   group=05_Tech Art Masks
   parameterName=ColorMask
   samplerSource=SSM_Wrap_WorldGroupSettings
   samplerType=SAMPLERTYPE_Masks
   sortPriority=0
   texture=&T_BlackMask
  }
  NamedRerouteDeclaration(None=SmoothStep(
    Min=ScalarParameter() {defaultValue=0.45, parameterName=Mask Step Min}
    Max=ScalarParameter() {defaultValue=0.5, parameterName=Mask Step Max}
    Value=StaticSwitchParameter(True=@n68, False=@n71.RGB) {group=01_SnowKnight_ColorCustomization, parameterName=Use TechArt Mask for Color, sortPriority=0}
   ) {constMax=0.5, constMin=0.45}) {name=FX, nodeColor=<rgba 0.8281 0 1 1>}
  NamedRerouteDeclaration(None=Min(A=NamedRerouteUsage(), B=NamedRerouteUsage())) {name=BannerAlpha, nodeColor=<rgba 1 0.9141 0 1>}
  NamedRerouteDeclaration(None=@n45.Metallic) {name=Metallic_In, nodeColor=<rgba 0.125 0 1 1>}
  n72 = NamedRerouteUsage()
  NamedRerouteDeclaration(None=ComponentMask(None=NamedRerouteDeclaration(None=@n72) {name=M, nodeColor=<rgba 0.8516 0 1 1>}) {b=true}) {name=ChainmailMask, nodeColor=<rgba 0 0.9297 1 1>}
  n73 = Saturate(None=Add(A=NamedRerouteUsage(), B=NamedRerouteUsage()))
  n74 = ScalarParameter() {group=01_SnowKnight_ColorCustomization, parameterName=BaseColorBias}
  NamedRerouteDeclaration(None=Add(A=Multiply(A=@n73, B=@n74), B=NamedRerouteUsage())) {name=BaseColorModified, nodeColor=<rgba 0 0.2734 1 1>}
  n75 = NamedRerouteUsage()
  n76 = CameraVectorWS()
  n77 = PixelNormalWS()
  n78 = CurveAtlasRowParameter(CurveTime=Clamp(None=DotProduct(A=@n76, B=@n77))) {
   atlas=/Game/Global/Textures/Default/Curve/FN_Character_Atlas
   curve=/Game/Characters/Player/Female/Medium/Bodies/F_MED_Ms_Alpine/Textures/AluminiumreflectanceCurve
   group=06_Metal LUT
   parameterName=Metal LUT Curve
  }
  n79 = ChannelMaskParameter(None=@n78) {group=06_Metal LUT, parameterName=LUTChannel}
  n80 = Saturate(None=ScalarParameter() {defaultValue=1, group=06_Metal LUT, parameterName=MetalLutIntensity})
  n81 = Saturate(None=ScalarParameter() {defaultValue=1, group=06_Metal LUT, parameterName=Secondary_MetalLutIntensity})
  n82 = NamedRerouteUsage()
  n83 = SetMaterialAttributes!(
   MaterialAttributes=@n3.MaterialAttributes
   Base Color=LinearInterpolate(A=@n3.Base Color, B=Multiply(A=@n3.Base Color, B=@n79.R), Alpha=LinearInterpolate(A=@n80, B=@n81, Alpha=@n82))
  ) {attributeSetTypes [#1]}
  n84 = NamedRerouteUsage()
  n85 = ChannelMaskParameter(None=@n84) {group=05_Tech Art Masks, parameterName=MetalLUTMaskChannel}
  n86 = Add(A=Add(A=ComponentMask(None=@n84) {r=true}, B=ComponentMask(None=@n84) {g=true}), B=ComponentMask(None=@n84) {b=true})
  n87 = LinearInterpolate(
   A=@n85.R
   B=Add(A=@n85.R, B=OneMinus(None=@n86))
   Alpha=ScalarParameter() {controlType=Enumeration, enumeration=&E_Mat_YesNo, parameterName=Use Pickaxe Metal Mask}
  )
  NamedRerouteDeclaration(None=StaticSwitchParameter(True=BlendMaterialAttributes(A=@n75, B=@n83, Alpha=@n87), False=NamedRerouteUsage()) {group=04_Tech Art, parameterName=UseMetalLUT}) {name=PreFxMaterial_Out, nodeColor=<rgba 0 0.6016 1 1>}
  n88 = StaticSwitchParameter(True=TextureCoordinate() {coordinateIndex=1}, False=TextureCoordinate()) {defaultValue=true, group=01_SnowKnight_BannerLogo, parameterName=UseUV1, sortPriority=1}
  n89 = AppendVector(
   A=ScalarParameter() {group=01_SnowKnight_BannerLogo, parameterName=Banner1_X}
   B=ScalarParameter() {group=01_SnowKnight_BannerLogo, parameterName=Banner1_Y}
  )
  n90 = MaterialFunctionCall(
   UVs=MaterialFunctionCall(UVs=Add(A=@n88, B=@n89), "Rotation Angle (0-1)"=ScalarParameter() {group=01_SnowKnight_BannerLogo, parameterName=Banner1_Rotation}) {materialFunction=&CustomRotator}.Rotated Values
   Texture Scale=ScalarParameter() {defaultValue=0.2, group=01_SnowKnight_BannerLogo, parameterName=Banner1_Scale}
  ) {materialFunction=&ScaleUVsByCenter}
  n91 = TextureSampleParameter2D(UVs=@n90.UVs) {
   group=01_SnowKnight_BannerLogo
   parameterName=Banner_Texture
   samplerSource=SSM_Clamp_WorldGroupSettings
   sortPriority=2
   texture=/Game/UI/Foundation/Textures/Banner/Shapes/banner_shape_standard_Mipmaps
  }
  NamedRerouteDeclaration(None=Multiply(A=@n91.R, B=@n90.0-1 mask)) {name=BannerLogo, nodeColor=<rgba 1 0 0.8672 1>}
  NamedRerouteDeclaration(None=ScalarParameter() {defaultValue=1, group=01_SnowKnight_BannerLogo, parameterName=LogoAlpha, sliderMax=1}) {name=LogoColor_Alpha, nodeColor=<rgba 1 0 0.4453 1>}
  NamedRerouteDeclaration(None=@n44.Base Color) {name=BaseColor, nodeColor=<rgba 0 0.6953 1 1>}
 ]
}
